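# Anim Blueprint for the CraneAnchor trap gadget head part, adjusting the part for Medium, Large and XLarge special cosmetic variants.
# The AnimGraph copies the attached parent's pose into a Control Rig, passing the three size flags as rig inputs.
# On customization load, item meta tags set those flags.
# A dynamic material parameter, Hide Element 01, is then cleared if any flag is true and set otherwise.
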AnimBlueprint /StrawTurn/Gameplay/CharacterPart/M_MED_StrawTurn/Meshes/Parts/Heads/CraneAnchor_Head_Trap_Gadget_AnimBP
target_skeleton=&/CraneAnchor_Head_Trap_Gadget_Skeleton
graphs [
 AnimGraph {nodes [
   n1 = Control Rig!(
    Source=Copy Pose From Mesh() {bUseAttachedParent=true}
    Alpha
    Special Medium:bool=Get Special Medium() -> (Special Medium:bool)
    Special Large:bool=Get Special Large() -> (Special Large:bool)
    Special XLarge:bool=Get Special XLarge() -> (Special XLarge:bool)
   ) -> (Pose) {
    alphaBoolBlend {blendOption=Linear}
    controlRigAssetReference {blueprintRigClass=&/CraneAnchor_Head_Trap_Gadget_CtrlRig.CraneAnchor_Head_Trap_Gadget_CtrlRig_C}
   }
   Output Pose(Result=@n1)
  ]}
 EventGraph {nodes [
   Event OnCharacterCustomizationLoaded()
   > n1 = Sequence()
   > Set Special Medium(Special Medium:bool=CharacterItemDefinitionHasAnyMetaTag(Tags=Make Array("[0]"=(TagName="Cosmetics.MetaTags.Special.Medium")))) -> (Output_Get:bool)
   > Set Special Large(Special Large:bool=CharacterItemDefinitionHasAnyMetaTag(Tags=Make Array("[0]"=(TagName="Cosmetics.MetaTags.Special.Large")))) -> (Output_Get:bool)
   > Set Special XLarge(Special XLarge:bool=CharacterItemDefinitionHasAnyMetaTag(Tags=Make Array("[0]"=(TagName="Cosmetics.MetaTags.Special.XLarge")))) -> (Output_Get:bool)
   n2 = Branch(execute=@n1.then_1, Condition=OR Boolean!(A=Get Special Large(), B=Get Special XLarge(), C:bool=Get Special Medium()) -> (ReturnValue))
   n3 = GetPartSkeletalMeshComponent()
   > n4 = Get Dynamic Material Instance(SkeletalMeshComponent=@n3)
   > SetScalarParameterValue(self=@n4, ParameterName=Hide Element 01, Value=0)
   n5 = Get Dynamic Material Instance(execute=@n2.else, SkeletalMeshComponent=@n3)
   > SetScalarParameterValue(self=@n5, ParameterName=Hide Element 01, Value=1)
  ]}
]
variables [
 Special Large {type=boolean}
 Special Medium {type=boolean}
 Special XLarge {type=boolean}
]
